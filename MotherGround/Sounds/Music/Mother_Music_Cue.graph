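# Layered music cue for the Mother encounter, routed through the NPC music sound class.
# A Mixer sums four looping stems (Idle, Base, Guard, Boss), with the Idle input set slightly lower.
# Each stem sits behind a continuous modulator whose volume follows a named parameter (Unaware, Base, Wall and Ground, respectively).
# These parameters default to zero, so layers rely on externally set values to become audible.
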SoundCue /MotherGround/Sounds/Music/Mother_Music_Cue
root=Mixer_1
nodes [
 {
  id=Mixer_1
  children [ModulatorContinuous_0, ModulatorContinuous_1, ModulatorContinuous_2, ModulatorContinuous_4]
  properties {inputVolume [0.7,1,1,1]}
 }
 {id=ModulatorContinuous_0, children [WavePlayer_3], properties {volumeModulationParams {default=0, parameterName=Unaware}}}
 {id=ModulatorContinuous_1, children [WavePlayer_0], properties {volumeModulationParams {default=0, parameterName=Base}}}
 {id=ModulatorContinuous_2, children [WavePlayer_2], properties {volumeModulationParams {default=0, parameterName=Wall}}}
 {id=ModulatorContinuous_4, children [WavePlayer_1], properties {volumeModulationParams {default=0, parameterName=Ground}}}
 {id=WavePlayer_0, properties {bLooping=true, soundWaveAssetPtr=&/S17_Mother_Music_Base}}
 {id=WavePlayer_1, properties {bLooping=true, soundWaveAssetPtr=&/S17_Mother_Music_Boss}}
 {id=WavePlayer_2, properties {bLooping=true, soundWaveAssetPtr=&/S17_Mother_Music_Guard}}
 {id=WavePlayer_3, properties {bLooping=true, soundWaveAssetPtr=&/S17_Mother_Music_Idle}}
]
properties {maxDistance=2097152, soundClassObject=&Fort_Music_NPC, volumeMultiplier=1}
